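; Quest gameplay ability that credits the Spiderman swing quest when the player chains enough swings without landing.
; On activation it zeroes SwingCount, then counts each time the grapple BlockUseAfterDetach tag is added to the owning pawn.
; A gameplay event listener on the tracked landing tag resets the count.
; Reaching SwingTarget sends a complex custom stat event to the quest manager, then ends the listener and ability.
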(event Ability|EventActivateAbilityFromEvent
  (params (EventData :type Gameplay%20Event%20Data%20Structure%20%28by%20ref%29))
  (bind (_eventtag _instigator _target _optionalobject _optionalobject2 _contexthandle _instigatortags _targettags) (Utilities|Struct|BreakGameplayEventData EventData))
  (|CallParentFunction EventData)
  (Variables|Tags|SetTargetTags _targettags)
  ;[1 0.2402 0.03963] Initialize Swing tracking...
  (Variables|Default|SetSwingCount 0)
  (bind _asynctaskproxy (Ability|Tasks|WaitGameplayEvent (Variables|Tags|GetTrackedLandingTag))
    (:then
      (Variables|Default|SetEventTask _asynctaskproxy)
      (Ability|Tasks|WaitGameplayTagAddToActor (Variables|GAVars|Data|GetOwningPlayerPawn) "(TagName=\"Granted.Athena.GrappleParent.BlockUseAfterDetach\")"
        (:Added
          (bind _result (Math|Integer|IncrementInt (Variables|Default|GetSwingCount))
            (:%20%20
              (bind _output_get (Variables|Default|SetSwingCount _result))
              (if (>= _output_get (Variables|Default|GetSwingTarget))
                ;[0.03963 0.9403 1] Max swing count reached!
                (Quest|SendComplexCustomStatEvent (Variables|GAVars|Data|GetOwningPlayerQuestManager) 0 (Variables|Tags|GetStatEventSourceTags) "(GameplayTags=)" 1)
                (Variables|Default|SetSwingCount 0)
                (GameplayTasks|EndTask (Variables|Default|GetEventTask))
                (Ability|EndAbility)))))))
    (:EventReceived
      ;[1 0.09399 0.7665] Target landed...
      (Variables|Default|SetSwingCount 0))))
